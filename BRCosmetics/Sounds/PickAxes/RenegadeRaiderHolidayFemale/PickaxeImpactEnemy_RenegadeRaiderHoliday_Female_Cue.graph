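# Sound cue for the female Renegade Raider Holiday pickaxe hitting an enemy, using melee impact attenuation and concurrency.
# A quality switch picks full six-variant pools at the first level, three-variant pools at the second and fourth, or one impact wave at the third.
# The pooled branches randomly play an impact alone or, less often, mix it with a Vox clip.
# Every layer passes through a modulator and a short fixed delay.
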SoundCue /BRCosmetics/Sounds/PickAxes/RenegadeRaiderHolidayFemale/PickaxeImpactEnemy_RenegadeRaiderHoliday_Female_Cue
root=QualityLevel_0
nodes [
 {id=Delay_0, children [Modulator_0], properties {delayMax=0.05, delayMin=0.05}}
 {id=Delay_1, children [Modulator_1], properties {delayMax=0.05, delayMin=0.05}}
 {id=Delay_2, children [Modulator_2], properties {delayMax=0.05, delayMin=0.05}}
 {id=Delay_3, children [Modulator_3], properties {delayMax=0.05, delayMin=0.05}}
 {id=Delay_4, children [Modulator_4], properties {delayMax=0.05, delayMin=0.05}}
 {id=Mixer_0, children [Delay_1,Delay_0], properties {inputVolume [1,1]}}
 {id=Mixer_1, children [Delay_2,Delay_4], properties {inputVolume [1,1]}}
 {id=Modulator_0, children [Random_4]}
 {id=Modulator_1, children [Random_2]}
 {id=Modulator_2, children [Random_3]}
 {id=Modulator_3, children [WavePlayer_16]}
 {id=Modulator_4, children [Random_6]}
 {id=QualityLevel_0, children [Random_5,Random_7,Delay_3,Random_7]}
 {
  id=Random_2
  children [WavePlayer_1, WavePlayer_2, WavePlayer_3, WavePlayer_4, WavePlayer_5, WavePlayer_12]
  properties {weights [1,1,1,1,1,1]}
 }
 {id=Random_3, children [WavePlayer_13, WavePlayer_14, WavePlayer_15], properties {weights [1,1,1]}}
 {
  id=Random_4
  children [WavePlayer_17, WavePlayer_18, WavePlayer_19, WavePlayer_20, WavePlayer_21, WavePlayer_22]
  properties {bRandomizeWithoutReplacement=false, weights [1,1,1,1,1,1]}
 }
 {id=Random_5, children [Delay_1,Mixer_0], properties {bRandomizeWithoutReplacement=false, weights [1,0.4]}}
 {id=Random_6, children [WavePlayer_23, WavePlayer_24, WavePlayer_25], properties {bRandomizeWithoutReplacement=false, weights [1,1,1]}}
 {id=Random_7, children [Delay_2,Mixer_1], properties {bRandomizeWithoutReplacement=false, weights [1,0.4]}}
 {
  id=WavePlayer_1
  properties {soundWaveAssetPtr=&/PickAxe_EnemyImpact_RenegadeRaider_Holiday_Female_06}
 }
 {
  id=WavePlayer_12
  properties {soundWaveAssetPtr=&/PickAxe_EnemyImpact_RenegadeRaider_Holiday_Female_05}
 }
 {
  id=WavePlayer_13
  properties {soundWaveAssetPtr=&/PickAxe_EnemyImpact_RenegadeRaider_Holiday_Female_01}
 }
 {
  id=WavePlayer_14
  properties {soundWaveAssetPtr=&/PickAxe_EnemyImpact_RenegadeRaider_Holiday_Female_03}
 }
 {
  id=WavePlayer_15
  properties {soundWaveAssetPtr=&/PickAxe_EnemyImpact_RenegadeRaider_Holiday_Female_02}
 }
 {
  id=WavePlayer_16
  properties {soundWaveAssetPtr=&PickAxe_EnemyImpact_RenegadeRaider_Holiday_Female_01}
 }
 {
  id=WavePlayer_17
  properties {soundWaveAssetPtr=&/PickAxe_EnemyImpact_RenegadeRaider_Holiday_Female_Vox_01}
 }
 {
  id=WavePlayer_18
  properties {soundWaveAssetPtr=&/PickAxe_EnemyImpact_RenegadeRaider_Holiday_Female_Vox_02}
 }
 {
  id=WavePlayer_19
  properties {soundWaveAssetPtr=&/PickAxe_EnemyImpact_RenegadeRaider_Holiday_Female_Vox_03}
 }
 {
  id=WavePlayer_2
  properties {soundWaveAssetPtr=&PickAxe_EnemyImpact_RenegadeRaider_Holiday_Female_01}
 }
 {
  id=WavePlayer_20
  properties {soundWaveAssetPtr=&/PickAxe_EnemyImpact_RenegadeRaider_Holiday_Female_Vox_04}
 }
 {
  id=WavePlayer_21
  properties {soundWaveAssetPtr=&/PickAxe_EnemyImpact_RenegadeRaider_Holiday_Female_Vox_05}
 }
 {
  id=WavePlayer_22
  properties {soundWaveAssetPtr=&/PickAxe_EnemyImpact_RenegadeRaider_Holiday_Female_Vox_06}
 }
 {
  id=WavePlayer_23
  properties {soundWaveAssetPtr=&PickAxe_EnemyImpact_RenegadeRaider_Holiday_Female_Vox_01}
 }
 {
  id=WavePlayer_24
  properties {soundWaveAssetPtr=&PickAxe_EnemyImpact_RenegadeRaider_Holiday_Female_Vox_02}
 }
 {
  id=WavePlayer_25
  properties {soundWaveAssetPtr=&PickAxe_EnemyImpact_RenegadeRaider_Holiday_Female_Vox_03}
 }
 {
  id=WavePlayer_3
  properties {soundWaveAssetPtr=&PickAxe_EnemyImpact_RenegadeRaider_Holiday_Female_02}
 }
 {
  id=WavePlayer_4
  properties {soundWaveAssetPtr=&PickAxe_EnemyImpact_RenegadeRaider_Holiday_Female_03}
 }
 {
  id=WavePlayer_5
  properties {soundWaveAssetPtr=&/PickAxe_EnemyImpact_RenegadeRaider_Holiday_Female_04}
 }
]
properties {
 attenuationSettings=/Game/Sounds/Attenuation/Fort_Pickaxe_Impact_Att
 audioPropertiesSheet=/BRCosmetics/Sounds/Settings/PropertySheets/Pickaxe/APS_Cosmetic_Pickaxe_Impact
 bPrimeOnLoad=true
 maxDistance=4300
 soundClassObject=&Fort_WeaponMelee_PSM
 volumeMultiplier=1
 concurrencySet [&Player_Melee_Weapon_Conc]
}
